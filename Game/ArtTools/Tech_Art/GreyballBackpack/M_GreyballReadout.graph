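# M_GreyballReadout is an unlit translucent material for skeletal meshes, apparently a brightness readout for the greyball backpack art tool.
# It splits scene color and takes the maximum of red and blue, with red fed twice and green unused.
# A very narrow SphereMask tests that value against the Brightness parameter, flagging near-exact matches.
# Neither the mask nor the red-orange color parameter is shown connected to a material output.
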Material /Game/ArtTools/Tech_Art/GreyballBackpack/M_GreyballReadout
graph {
 properties {
  bUsedWithMeshDeformer=true
  bUsedWithSkeletalMesh=true
  blendMode=BLEND_Translucent
  pixelDepthOffsetMode=PDOM_Legacy
  shadingModel=MSM_Unlit
  usedShadingModels=Unlit
 }
 nodes [
  n1 = MaterialFunctionCall(Color=SceneColor!(UVs).RGB) {materialFunction=&SplitComponents}
  SphereMask(A=ScalarParameter() {defaultValue=0.115, parameterName=Brightness}, B=Max(A=Max(A=@n1.R, B=@n1.R), B=@n1.B)) {attenuationRadius=0.001}
  VectorParameter() {defaultValue=<rgba 1 0.0089 0 0>}
 ]
}
