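# Greyball backpack base material in the tech art tools: a Default Lit surface that blends flat grey with a small colored dot, red by default.
# A sphere mask measures UV distance from the red and green channels of the Positional parameter, placing the dot in texture space.
# That mask drives the grey-to-color lerp.
# Usage flags cover skeletal meshes, mesh deformers and Nanite.
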Material /Game/ArtTools/Tech_Art/GreyballBackpack/M_Greyball_Base
graph {
 properties {
  bUsedWithMeshDeformer=true
  bUsedWithNanite=true
  bUsedWithSkeletalMesh=true
  pixelDepthOffsetMode=PDOM_Legacy
  usedShadingModels=Default Lit
 }
 nodes [
  Constant()
  Constant() {r=0.1}
  Constant() {r=0.5}
  n1 = VectorParameter() {parameterName=Positional, defaultValue=<rgba 0.863 0.2945 0 1>}
  LinearInterpolate(
   A=Constant() {r=0.5}
   B=VectorParameter() {defaultValue=<rgba 1 0 0 0>}.RGB
   Alpha=SphereMask(A=TextureCoordinate(), B=ComponentMask(None=@n1.RGB) {g=true, r=true}) {attenuationRadius=0.02}
  )
 ]
}
