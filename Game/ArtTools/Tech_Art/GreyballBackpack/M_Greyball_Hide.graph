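# Masked, unlit material in the Greyball backpack tech-art folder, named as a hide material, apparently for making mesh sections invisible.
# The graph holds only a zero constant and a red-default vector parameter, with no pin connections listed.
# It is flagged for use on skeletal meshes and with mesh deformers.
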Material /Game/ArtTools/Tech_Art/GreyballBackpack/M_Greyball_Hide
graph {
 properties {
  bUsedWithMeshDeformer=true
  bUsedWithSkeletalMesh=true
  blendMode=BLEND_Masked
  pixelDepthOffsetMode=PDOM_Legacy
  shadingModel=MSM_Unlit
  usedShadingModels=Unlit
 }
 nodes [
  Constant()
  VectorParameter() {defaultValue=<rgba 1 0 0 0>}
 ]
}
