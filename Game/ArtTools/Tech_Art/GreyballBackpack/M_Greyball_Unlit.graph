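# M_Greyball_Unlit is a two-sided unlit material in the GreyballBackpack tech-art folder, apparently a flat grey for that tool.
# Its graph holds a single dark scalar constant, with no textures, parameters or math, and no visible link to an output.
# It is flagged for use on skeletal meshes and with mesh deformers.
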Material /Game/ArtTools/Tech_Art/GreyballBackpack/M_Greyball_Unlit
graph {
 properties {
  bUsedWithMeshDeformer=true
  bUsedWithSkeletalMesh=true
  pixelDepthOffsetMode=PDOM_Legacy
  shadingModel=MSM_Unlit
  twoSided=true
  usedShadingModels=Unlit
 }
 nodes [Constant() {r=0.115}]
}
