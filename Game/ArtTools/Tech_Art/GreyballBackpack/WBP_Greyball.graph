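# WBP_Greyball is a Tech Art debug widget with a BugIt cheat button, player location readout and time-of-day slider.
# Tick refreshes location and time text, syncing the slider to game time unless IsCapturing is set.
# Mouse or controller capture sets that flag and prints a leftover "Hello".
# While captured, slider changes send SetTimeOfDay cheat commands; release calls SetTimeOfDay directly and clears the flag.
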WidgetBlueprint /Game/ArtTools/Tech_Art/GreyballBackpack/WBP_Greyball
graphs [EventGraph {nodes [
   "On Clicked (Bugit_Button)"!() -> (OutputDelegate, then, widget:User Widget Object Reference, event:Pointer Event Structure) {is=title K2Node_ComponentBoundEvent}
   > ExecuteConsoleCommand!(execute, Command=cheat Bugit, SpecificPlayer) -> (then)
   n1 = "On Controller Capture Begin (ToD_Slider)"() {is=title K2Node_ComponentBoundEvent}
   Set IsCapturing(execute=[@n1.then, @n3.then], IsCapturing:bool=true) -> (Output_Get:bool)
   > PrintString!(execute, InString=Hello, bPrintToScreen, bPrintToLog, TextColor, Duration, Key) -> (then)
   n2 = "On Controller Capture End (ToD_Slider)"() {is=title K2Node_ComponentBoundEvent}
   SetTimeOfDay!(
    execute=[@n2.then, @n4.then]
    TimeOfDay:float=GetValue!(self:Slider Object Reference=Get ToD_Slider() -> (ToD_Slider:Slider Object Reference)) -> (ReturnValue:float)
   ) -> (then)
   > Set IsCapturing(IsCapturing=false)
   n3 = "On Mouse Capture Begin (ToD_Slider)"() {is=title K2Node_ComponentBoundEvent}
   n4 = "On Mouse Capture End (ToD_Slider)"() {is=title K2Node_ComponentBoundEvent}
   n5 = "On Value Changed (ToD_Slider)"!() -> (OutputDelegate, then, Value:float) {is=title K2Node_ComponentBoundEvent}
   > Branch(Condition=Get IsCapturing() -> (IsCapturing:bool))
   > ExecuteConsoleCommand(Command=Append(A="Cheat SetTimeOfDay ", B="To String (Float)"(InDouble=@n5.Value)))
   Event Tick()
   n6 = GetPlayerPawn!(PlayerIndex:int=0) -> (ReturnValue:Pawn Object Reference)
   > "SetText (Text)"(
    self=Get Loc() -> (Loc:Common Text Object Reference)
    InText="To Text (Vector)"!(InVec:Vector=Get Actor Location(self=@n6)) -> (ReturnValue:Text)
   )
   > n7 = Branch(Condition=Get IsCapturing())
   n8 = Get ToD_Slider()
   "SetText (Text)"(
    execute=[@n9.then, @n7.then]
    self=Get TimeOfDay() -> (TimeOfDay:Common Text Object Reference)
    InText="To Text (Float)"(Value=GetValue(self=@n8), MinimumIntegralDigits=1, MaximumIntegralDigits=324, MinimumFractionalDigits=0, MaximumFractionalDigits=3)
   )
   n9 = SetValue!(execute=@n7.else, self:Slider Object Reference=Get ToD_Slider(), InValue:float=GetTimeOfDay!() -> (ReturnValue:float)) -> (then)
   Event PreConstruct()
   Event Construct()
  ]}]
variables [IsCapturing {type=boolean}]
widgets {
 info {inheritedWidgetCount=0, parentClass=&UserWidget, rootWidgetClass=&Overlay, widgetCount=10}
 widgets [
  Overlay_24 {widgetClassPath=&Overlay, widget=Overlay_24}
  VerticalBox_28 {
   widgetClassPath=&VerticalBox
   parent=Overlay_24
   widget=VerticalBox_28
   slot {class=OverlaySlot, properties {horizontalAlignment=HAlign_Fill, verticalAlignment=VAlign_Fill}}
  }
  HorizontalBox_28 {
   widgetClassPath=&HorizontalBox
   parent=VerticalBox_28
   widget=HorizontalBox_28
   slot {class=VerticalBoxSlot, properties {verticalAlignment=VAlign_Top}}
  }
  Bugit_Button {
   widgetClassPath=&TA_Generic01_Button
   parent=HorizontalBox_28
   widget=Bugit_Button
   properties {label=BugIt}
   slot {class=HorizontalBoxSlot}
  }
  Loc {
   widgetClassPath=&CommonTextBlock
   bIsVariable=true
   parent=HorizontalBox_28
   widget=Loc
   properties {
    style=/FortUILibrary/Styles/Typography/Foundation/TextStyle-Base-S.TextStyle-Base-S_C
    text=Location
    font {
     bForceMonospaced=false
     bMaterialIsStencil=false
     fontMaterial=null
     fontObject=&NotoSans
     letterSpacing=0
     monospacedWidth=1
     size=18
     skewAmount=0
     typefaceFontName=NotoSans-Regular
     outlineSettings {
      bApplyOutlineToDropShadows=false
      bMiteredCorners=false
      bSeparateFillAlpha=false
      outlineMaterial=null
      outlineSize=0
      outlineColor=<rgba 0 0 0 1>
     }
    }
    shadowOffset=<0 0>
   }
   slot {class=HorizontalBoxSlot, properties {horizontalAlignment=HAlign_Left}}
  }
  HorizontalBox_146 {widgetClassPath=&HorizontalBox, parent=VerticalBox_28, widget=HorizontalBox_146, slot {class=VerticalBoxSlot}}
  CommonTextBlock_187 {
   widgetClassPath=&CommonTextBlock
   parent=HorizontalBox_146
   widget=CommonTextBlock_187
   properties {
    style=&TextStyle-Base_C
    text=ToD:
    font {
     bForceMonospaced=false
     bMaterialIsStencil=false
     fontMaterial=null
     fontObject=&NotoSans
     letterSpacing=0
     monospacedWidth=1
     size=24
     skewAmount=0
     typefaceFontName=NotoSans-Regular
     outlineSettings {
      bApplyOutlineToDropShadows=false
      bMiteredCorners=false
      bSeparateFillAlpha=false
      outlineMaterial=null
      outlineSize=0
      outlineColor=<rgba 0 0 0 1>
     }
    }
    shadowOffset=<0 0>
   }
   slot {class=HorizontalBoxSlot}
  }
  TimeOfDay {
   widgetClassPath=&CommonTextBlock
   bIsVariable=true
   parent=HorizontalBox_146
   widget=TimeOfDay
   properties {
    style=&TextStyle-Base_C
    text=Text Block
    font {
     bForceMonospaced=false
     bMaterialIsStencil=false
     fontMaterial=null
     fontObject=&NotoSans
     letterSpacing=0
     monospacedWidth=1
     size=24
     skewAmount=0
     typefaceFontName=NotoSans-Regular
     outlineSettings {
      bApplyOutlineToDropShadows=false
      bMiteredCorners=false
      bSeparateFillAlpha=false
      outlineMaterial=null
      outlineSize=0
      outlineColor=<rgba 0 0 0 1>
     }
    }
    shadowOffset=<0 0>
   }
   slot {class=HorizontalBoxSlot, properties {padding=<ltrb 4 2 4 2>}}
  }
  HorizontalBox_193 {widgetClassPath=&HorizontalBox, parent=VerticalBox_28, widget=HorizontalBox_193, slot {class=VerticalBoxSlot}}
  ToD_Slider {
   widgetClassPath=/Script/UMG.Slider
   bIsVariable=true
   parent=HorizontalBox_193
   widget=ToD_Slider
   properties {maxValue=23.99}
   slot {class=HorizontalBoxSlot, properties {size {sizeRule=Fill, value=1}}}
  }
 ]
}
